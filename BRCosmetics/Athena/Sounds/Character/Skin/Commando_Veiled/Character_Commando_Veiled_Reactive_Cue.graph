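# Sound cue for the reactive sound of the Commando Veiled character skin, playing kill-reaction clips.
# A random node picks one of six equally weighted KillReact waves, passed through a modulator for slight pitch and volume variation, under a root delay node.
# It uses basic foley attenuation and the gadgets sound class.
# A concurrency override allows one instance at a time, stopping the oldest.
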SoundCue /BRCosmetics/Athena/Sounds/Character/Skin/Commando_Veiled/Character_Commando_Veiled_Reactive_Cue
root=Delay_0
nodes [
 {id=Delay_0, children [Modulator_0]}
 {id=Modulator_0, children [Random_1], properties {pitchMax=1.1, volumeMax=1.1, volumeMin=1}}
 {id=Random_1, children [WavePlayer_0,WavePlayer_1,WavePlayer_2,WavePlayer_3,WavePlayer_4,WavePlayer_5], properties {weights [1,1,1,1,1,1]}}
 {id=WavePlayer_0, properties {soundWaveAssetPtr=/Game/Athena/Sounds/Character/Skin/Commando_Veiled/Veiled_KillReact_1}}
 {id=WavePlayer_1, properties {soundWaveAssetPtr=/Game/Athena/Sounds/Character/Skin/Commando_Veiled/Veiled_KillReact_2}}
 {id=WavePlayer_2, properties {soundWaveAssetPtr=/Game/Athena/Sounds/Character/Skin/Commando_Veiled/Veiled_KillReact_3}}
 {id=WavePlayer_3, properties {soundWaveAssetPtr=/Game/Athena/Sounds/Character/Skin/Commando_Veiled/Veiled_KillReact_4}}
 {id=WavePlayer_4, properties {soundWaveAssetPtr=/Game/Athena/Sounds/Character/Skin/Commando_Veiled/Veiled_KillReact_5}}
 {id=WavePlayer_5, properties {soundWaveAssetPtr=/Game/Athena/Sounds/Character/Skin/Commando_Veiled/Veiled_KillReact_6}}
]
properties {
 attenuationSettings=&Foley_Attenuation_Basic_Att
 bOverrideConcurrency=true
 maxDistance=1536
 soundClassObject=&Fort_Gadgets
 volumeMultiplier=0.8
 concurrencyOverrides {maxCount=1, resolutionRule=StopOldest}
}
